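; Gameplay ability for the Emerald "Damage After Potafort" quest that marks the owning player with an effect, apparently for damage-after-use tracking.
; On event activation it calls the parent, then checks the owning pawn for the equipped BigLeafGrenade consumable tag.
; If the tag is present, it applies the GE_Quest_AfterPAF effect to the pawn's own ability system component.
; Only the branch without the tag explicitly ends the ability.
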(event Ability|EventActivateAbilityFromEvent
  (params (EventData :type Gameplay%20Event%20Data%20Structure%20%28by%20ref%29))
  (bind _owningplayerpawn (Variables|GAVars|Data|GetOwningPlayerPawn))
  (|CallParentFunction EventData)
  (bind _bdidactorhavetags (Class|AthenaQuestFunctionLibrary|AthenaCheckActorHasTags _owningplayerpawn "(GameplayTags=((TagName=\"Granted.ItemEquipped.Consumable.BigLeafGrenade.VR\")))"))
  (if _bdidactorhavetags
    (GameplayEffects|ApplyGameplayEffectToSelf :self (Ability|GetFortAbilitySystemComponent _owningplayerpawn) :GameplayEffectClass (Variables|Default|GetGE_Quest_AfterPAF) :Level 0.0)
    (Ability|EndAbility)))
